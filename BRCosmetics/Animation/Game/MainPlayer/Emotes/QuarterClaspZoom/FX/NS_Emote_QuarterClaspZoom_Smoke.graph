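# Niagara smoke effect for the QuarterClaspZoom emote, drawn as a single ribbon emitter.
# Continuously spawned particles rise from an upward offset, tinted by time-of-day color and narrowing along a width curve.
# An emitter-age curve over eight seconds delays, holds and fades overall opacity, multiplied with a per-particle alpha fade-in and fade-out.
# User parameters are declared, but none are referenced in the emitter modules shown.
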NiagaraSystem /BRCosmetics/Animation/Game/MainPlayer/Emotes/QuarterClaspZoom/FX/NS_Emote_QuarterClaspZoom_Smoke
dependencies {usedDataInterfaces [&NiagaraDataInterfaceCurve], usedDynamicInputs [], usedModules [&SystemState]}
emitters {Minimal {
  emitterUpdateScript [
   &EmitterState(Life Cycle Mode=::NewEnumerator1, Loop Duration=1, Scalability Mode=::NewEnumerator1, UseLoopDelay=false)
   &SpawnRate(SpawnRate=30, Use Spawn Probability=false)
   /SetVariables(Emitter.OpacityScalar:NiagaraFloat=&FloatFromCurve(
     FloatCurve=DataInterface() {propertyValues {Curve {keys [<at 0 0>,<at 0.35 0>,<at 0.45 1>,<at 0.9 1>,<at 1 0>]}}}
     CurveIndex=&ClampFloat(Float=&Divide_Float(A=$Emitter.Age, B=8))
    )) {isSetParameters=true}
  ]
  particleSpawnScript [
   &InitializeParticle(
    Color=&FNTimeOfDayColor(
     Night Color=<rgba 0.0145 0.023 0.0417 2>
     Morning Color=<rgba 0.0417 0.0417 0.0417 2>
     Day Color=<rgba 0.0417 0.0417 0.0417 2>
     Evening Color=<rgba 0.0417 0.0417 0.0417 2>
    )
    Lifetime=3
    Position Offset=<0 0 25>
    Ribbon Width=60
    Ribbon Width Mode=::NewEnumerator1
    UsePositionOffset=true
    Write Material Random=true
   )
   &AddVelocity(Velocity=<0 0 30>)
  ]
  particleUpdateScript [
   &ParticleState()
   &SolveForcesAndVelocity(Clamp Velocity=false, Limit Acceleration=false)
   &ScaleRibbonWidth(Ribbon Width Scale=&FloatFromCurve(FloatCurve=DataInterface() {propertyValues {Curve {keys [<at 0 1>,<at 1 0.2>]}}}))
   &ScaleColor(Scale Alpha=$Emitter.OpacityScalar, ScaleA=true, ScaleRGB=true)
   &ScaleColor(
    Scale Alpha=&FloatFromCurve(FloatCurve=DataInterface() {propertyValues {Curve {keys [<at 0 0>,<at 0.1032 1>,<at 0.8108 1>,<at 1 0>]}}})
    ScaleA=true
    ScaleRGB=true
   ) {module=ScaleColor001}
  ]
  rendererClasses [&NiagaraRibbonRendererProperties]
  renderers [{rendererClass=&NiagaraRibbonRendererProperties}]
 }}
summary {
 systemName=NS_Emote_QuarterClaspZoom_Smoke
 emitters [{emitterName=Minimal, rendererClasses [&NiagaraRibbonRendererProperties]}]
 userVariables [
  User.Anim_LocalSpaceInheritance {type=NiagaraFloat, defaultValue=1, description=""}
  User.BaseColorAlpha {type=LinearColor, description="", defaultValue {struct=LinearColor, value=<rgba 1 1 1 1>}}
  User.EmissiveColor {type=LinearColor, description="", defaultValue {struct=LinearColor, value=<rgba 1 1 1 0.761>}}
  User.Global_Smoke_Opacity_Multiplier {type=NiagaraFloat, defaultValue=0, description=""}
  User.NormalizedNotifyProgress {type=NiagaraFloat, defaultValue=0, description=""}
  User.OwnerLocalVelocity {type=Vector3f, description="", defaultValue {struct=Vector3f, value=<0 0 0>}}
  User.RibbonIDOffset {type=NiagaraInt32}
  User.RibbonScale {type=NiagaraFloat, defaultValue=1.5, description=""}
  User.Smoke_Alpha_B {type=NiagaraFloat, defaultValue=1, description=""}
  User.Smoke_Large_Offset {type=NiagaraFloat, defaultValue=0, description=""}
  User.UseEmissiveSmoke {type=NiagaraBool}
 ]
}
